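# Niagara system for the LoveSpin guitar cosmetic's idle wind effect, continuously spawning pink spinning blur and airflow at one offset point.
# BladeBlur spins sprites rapidly, Windwhirl rotates growing meshes, and ExhaustSmoke plays SubUV puffs in two mirrored, opposite-facing spawn groups.
# BladeBlur and Windwhirl adjust for time of day and User.bIsFrontEndPreview.
# The NE_DarkDisk emitter is disabled.
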NiagaraSystem /SparksCosmetics/Instrument/Guitar/LoveSpin/Effects/NS_Guitar_LoveSpin_Idle_Wind
dependencies {
 usedDataInterfaces [&NiagaraDataInterfaceColorCurve, &NiagaraDataInterfaceCurve]
 usedDynamicInputs []
 usedModules [&SystemState]
}
emitters {
 BladeBlur {
  emitterUpdateScript [
   &EmitterState()
   &SpawnRate(SpawnRate=8, Use Spawn Probability=false)
  ]
  particleSpawnScript [
   &InitializeParticle(
    Color Maximum=<rgba 1 0.2452 0.3584 0.5>
    Color Minimum=<rgba 0.7061 0.0454 0.1445 0.5>
    Color Mode=::NewEnumerator2
    Lifetime Max=1
    Lifetime Min=0.5
    Lifetime Mode=::NewEnumerator1
    Position Offset=<0 -2.75 -12>
    Sprite Rotation Mode=::NewEnumerator2
    Sprite Size Mode=::NewEnumerator3
    Uniform Sprite Size=35
    UsePositionOffset=true
    Write Material Random=true
   )
   &DynamicMaterialParameters(
    Index 0 Param 1=-60
    Module.Param0WriteEnabled=true
    Module.Param1WriteEnabled=true
    Module.Param2WriteEnabled=true
    Module.Param3WriteEnabled=true
   )
   &CameraOffset(Camera Offset Amount=-10)
  ]
  particleUpdateScript [
   &ParticleState()
   &SolveForcesAndVelocity(Clamp Velocity=false, Limit Acceleration=false)
   /SetVariables(Particles.SpriteFacing=&RotateVectorByQuaternion(Vector To Rotate=<0 1 0>, Quaternion=$Particles.MeshOrientation)) {isSetParameters=true}
   &SpriteRotationRate(Rotation Rate=&Multiply_Float(A=&UniformRangedFloat(Minimum=1, Maximum=4.5), B=500))
   &ScaleColor(
    Scale Alpha=&FloatFromCurve(
     FloatCurve=DataInterface() {propertyValues {Curve {keys [{arriveTangent=18.94, interpMode=RCIM_Cubic, leaveTangent=18.94, tangentMode=RCTM_User, time=0, value=0}, <at 0.1 1 RCIM_Cubic>, <at 1 0>]}}}
     Scale Curve=&Multiply_Float(
      A=&FNTimeOfDayScalar(Night Scale=0.5, Morning Scale=0.6, Evening Scale=0.6)
      B=&MakeCustomFloatFromBool(Bool=$User.bIsFrontEndPreview, True Float=0.3, False Float=1)
     )
    )
    ScaleA=true
    ScaleRGB=true
   )
  ]
  rendererClasses [&NiagaraSpriteRendererProperties, &NiagaraMeshRendererProperties]
  renderers [
   {rendererClass=&NiagaraSpriteRendererProperties}
   {rendererClass=&NiagaraMeshRendererProperties, rendererIndex=1}
  ]
 }
 ExhaustSmoke {
  emitterUpdateScript [
   &EmitterState()
   &SpawnRate(SpawnRate=4, Use Spawn Probability=false)
   &SpawnRate(SpawnGroup=1, SpawnRate=4, Use Spawn Probability=false) {module=SpawnRate001}
  ]
  particleSpawnScript [
   &InitializeParticle(
    Color=<rgba 1.5 0.7017 0.7018 1>
    Lifetime=&RandomRangeFloat(Minimum=0.75, Maximum=1)
    Position Offset=<0 -2.75 -12>
    Position Offset Coordinate Space=::Local
    Sprite Rotation Mode=::NewEnumerator2
    Sprite Size Mode=::NewEnumerator3
    Sprite UV Mode=::NewEnumerator4
    Sprite UV Scale:Vector2f=&MaskVector2dBySpawnGroup(Masked Vector2D=<1 1>, Passthrough Vector2D=<-1 1>, Spawn Group Mask=0)
    Uniform Sprite Size=&RandomRangeFloat(Minimum=20, Maximum=30)
    UsePositionOffset=true
    Write Material Random=true
   )
   /SetVariables(Particles.SubUVRandomStart:NiagaraInt32=&RandomRangeInt(Minimum=0, Maximum=44)) {isSetParameters=true}
  ]
  particleUpdateScript [
   &ParticleState()
   &ScaleColor(
    Scale Alpha=&FloatFromCurve(
     FloatCurve=DataInterface() {propertyValues {Curve {keys [
         {arriveTangent=8.9237, interpMode=RCIM_Cubic, leaveTangent=8.9237, tangentMode=RCTM_User, time=0, value=0}
         {arriveTangent=-3.3421, interpMode=RCIM_Cubic, leaveTangent=-3.3421, tangentMode=RCTM_User, time=0.2, value=1}
         <at 1 0 RCIM_Cubic>
        ]}}}
     Scale Curve=2
    )
    ScaleA=true
    ScaleRGB=false
   )
   &SpriteRotationRate(Rotation Rate=&FloatFromCurve(
     FloatCurve=DataInterface() {propertyValues {Curve {keys [
         {arriveTangent=-4.0107, interpMode=RCIM_Cubic, leaveTangent=-4.0107, tangentMode=RCTM_User, time=-0.0046, value=1.0113}
         {arriveTangent=-0.2496, interpMode=RCIM_Cubic, leaveTangent=-0.2496, tangentMode=RCTM_User, time=0.5, value=0.16}
        ]}}}
     Scale Curve=&Multiply_Float(
      A=&RandomRangeFloat(Minimum=-50, Maximum=-50)
      B=&MaskFloatBySpawnGroup(Masked Float=-1, Passthrough Float=0.5)
     )
    ))
   &ScaleSpriteSize(Scale Factor=&Vector2DFromFloat(Value=&FloatFromCurve(FloatCurve=DataInterface() {propertyValues {Curve {keys [
          {arriveTangent=1.4788, interpMode=RCIM_Cubic, leaveTangent=1.4788, tangentMode=RCTM_User, time=0, value=0.4}
          {arriveTangent=0.2754, interpMode=RCIM_Cubic, leaveTangent=0.2754, tangentMode=RCTM_User, time=1.0463, value=1}
         ]}}})))
   &SubUVAnimation(
    End Frame:NiagaraInt32=&Add_Integer(A=$Particles.SubUVRandomStart, B=23)
    Start Frame:NiagaraInt32=$Particles.SubUVRandomStart
    SubUV Animation Mode:ENiagara_SubUVLookupMode=::NewEnumerator1
    SubUV Curve=DataInterface() {propertyValues {Curve {keys [
        {arriveTangent=1.6197, interpMode=RCIM_Cubic, leaveTangent=1.6197, tangentMode=RCTM_User, time=0, value=0}
        {arriveTangent=0.5634, interpMode=RCIM_Cubic, leaveTangent=0.5634, tangentMode=RCTM_User, time=1, value=1}
       ]}}}
    SubUV Loop Count=0.5
   )
   &SpriteFacingAndAlignment(
    Sprite Facing=&MaskVectorBySpawnGroup(Masked Vector=<0 1 0>, Passthrough Vector=<0 -1 0>)
    WriteAlignment=false
    WriteFacing=true
   )
  ]
  rendererClasses [&NiagaraSpriteRendererProperties]
  renderers [{rendererClass=&NiagaraSpriteRendererProperties}]
 }
 NE_DarkDisk {
  bEnabled=false
  emitterUpdateScript [
   &EmitterState()
   &SpawnRate(SpawnRate=20, Use Spawn Probability=false) {enabled=false}
   &SpawnBurst_Instantaneous(Use Spawn Probability=false, UseLoopCountLimit=true)
  ]
  particleSpawnScript [
   &InitializeParticle(
    Lifetime=1.7
    Mesh Scale Mode=::NewEnumerator3
    Mesh Uniform Scale=0.12
    Position Offset=<0 -2.75 -12>
    Sprite Size Mode=::NewEnumerator3
    Uniform Sprite Size=5
    UsePositionOffset=true
    Write Material Random=true
   )
   &Color()
   &InitialMeshOrientation(Mesh Axis To Orient=<0 1 0>, Mesh Orientation Mode=::NewEnumerator2, Use Rotation Vector=false)
  ]
  particleUpdateScript [
   &ParticleState(Kill Particles When Lifetime Has Elapsed=false)
   &SolveForcesAndVelocity(Clamp Velocity=false, Limit Acceleration=false)
   &ScaleColor(
    Scale Alpha=&FloatFromCurve(FloatCurve=DataInterface() {propertyValues {Curve {keys [<at 0 1 RCIM_Cubic>,<at 1 0>]}}})
    ScaleA=true
    ScaleRGB=true
   )
  ]
  rendererClasses [&NiagaraSpriteRendererProperties, &NiagaraMeshRendererProperties]
  renderers [
   {rendererClass=&NiagaraSpriteRendererProperties}
   {rendererClass=&NiagaraMeshRendererProperties, rendererIndex=1}
  ]
 }
 Windwhirl {
  emitterUpdateScript [
   &EmitterState()
   &SpawnRate(SpawnRate=10, Use Spawn Probability=false)
  ]
  particleSpawnScript [
   &InitializeParticle(
    Color=<rgba 1 0.0783 0.665 0.25>
    Lifetime Max=1
    Lifetime Mode=::NewEnumerator1
    Mesh Scale Mode=::NewEnumerator1
    Mesh Uniform Scale Max=0.17
    Mesh Uniform Scale Min=0.15
    Position Offset=<0 -2.75 -12>
    Position Offset Coordinate Space=::Local
    UsePositionOffset=true
    Write Material Random=true
   )
   &ShapeLocation(
    Cylinder Height=10
    Cylinder Radius=5
    MaskSpawnGroup=false
    Override Seed=false
    Rotation Angle=90
    Rotation Axis=<0 1 0>
    Rotation Mode=::NewEnumerator1
    Shape Primitive=::NewEnumerator1
    Surface Only=false
    UseOwnerScale=true
   ) {enabled=false}
   &InitialMeshOrientation(
    Mesh Axis To Orient=<0 1 0>
    Mesh Orientation Mode=::NewEnumerator2
    Orientation Vector=<0 0 1>
    Rotation=&RandomRangeVector(Minimum=<0 0 0>, Maximum=<0 0 1>)
    Use Rotation Vector=true
   )
  ]
  particleUpdateScript [
   &ParticleState()
   &ScaleColor(
    Scale Alpha=&FloatFromCurve(
     FloatCurve=DataInterface() {propertyValues {Curve {keys [
         {arriveTangent=9.0102, interpMode=RCIM_Cubic, leaveTangent=9.0102, tangentMode=RCTM_User, time=0, value=0}
         {arriveTangent=2.4457, interpMode=RCIM_Cubic, leaveTangent=2.4457, tangentMode=RCTM_User, time=0.1, value=0.75}
         {arriveTangentWeight=0.1333, interpMode=RCIM_Cubic, leaveTangentWeight=0.1667, tangentMode=RCTM_User, time=0.5, value=1}
         {arriveTangent=-3.2028, interpMode=RCIM_Cubic, leaveTangent=-3.2028, tangentMode=RCTM_User, time=1, value=0}
        ]}}}
     Scale Curve=&FNTimeOfDayScalar(Night Scale=1, Morning Scale=0.8, Day Scale=0.8, Evening Scale=0.5)
    )
    Scale RGB=&MakeVectorFromLinearColorRGB(Color=&MakeCustomLinearColorFromBool(
      Bool=$User.bIsFrontEndPreview
      True Color=<rgba 0.8052 0.8383 1 1>
      False Color=&ColorFromCurve(
       ColorCurve=DataInterface() {propertyValues {
         AlphaCurve {keys [<at 0 1 RCIM_Cubic>]}
         BlueCurve {keys [<at 0.07 1 RCIM_Cubic>,<at 0.125 0.6>,<at 0.25 0.7>,<at 0.5 1>,<at 0.625 0.9>,<at 0.75 0.9891>,<at 1 1 RCIM_Cubic>]}
         GreenCurve {keys [<at 0.07 0.4436 RCIM_Cubic>,<at 0.125 0.7169>,<at 0.25 0.8561>,<at 0.5 1>,<at 0.625 0.9522>,<at 0.75 0.8087>,<at 1 0.4436 RCIM_Cubic>]}
         RedCurve {keys [<at 0.07 0.04 RCIM_Cubic>,<at 0.125 1>,<at 0.25 1>,<at 0.5 1>,<at 0.625 1>,<at 0.75 1>,<at 1 0.04 RCIM_Cubic>]}
        }}
       CurveIndex=Linked() {linkedVariable {name=NPC.FortniteNPC.TimeOfDay_0-1, type=NiagaraFloat}}
      )
     ))
    ScaleA=true
    ScaleRGB=true
   )
   &UpdateMeshOrientation(
    Rotation Coordinate Space=::NewEnumerator3
    Rotation Rate=&RandomRangeFloat(Minimum=2, Maximum=5)
    Rotation Vector=&RandomRangeVector(Minimum=<0 0 -1>, Maximum=<0 0 0>)
   )
   &ScaleMeshSize(Scale Factor=&Multiply_VectorByFloat(
     Vector=<1 1 1>
     Float=&FloatFromCurve(FloatCurve=DataInterface() {propertyValues {Curve {keys [
          {arriveTangent=2.3316, interpMode=RCIM_Cubic, leaveTangent=2.3316, tangentMode=RCTM_User, time=0, value=0}
          {arriveTangent=0.8383, interpMode=RCIM_Cubic, leaveTangent=0.8383, tangentMode=RCTM_User, time=0.6, value=1}
         ]}}})
    ))
   /SetVariables(Particles.Color=<rgba 1 1 1 10>) {enabled=false, isSetParameters=true}
  ]
  rendererClasses [&NiagaraMeshRendererProperties]
  renderers [{rendererClass=&NiagaraMeshRendererProperties, rendererIndex=0}]
 }
}
summary {
 systemName=NS_Guitar_LoveSpin_Idle_Wind
 emitters [
  {emitterName=Windwhirl, rendererClasses [&NiagaraMeshRendererProperties]}
  {emitterName=ExhaustSmoke, rendererClasses [&NiagaraSpriteRendererProperties]}
  {
   bEnabled=false
   emitterName=NE_DarkDisk
   rendererClasses [&NiagaraSpriteRendererProperties, &NiagaraMeshRendererProperties]
  }
  {emitterName=BladeBlur, rendererClasses [&NiagaraSpriteRendererProperties, &NiagaraMeshRendererProperties]}
 ]
 userVariables [User.bIsFrontEndPreview {type=NiagaraBool}]
}
