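# Animation Blueprint for the LoveSpin guitar that continuously spins its dyn_wheel bone.
# Each update eases a speed value toward a fixed negative target with FInterpTo, then adds it, scaled by delta time and a constant, to a wrapped 0-360 angle.
# That angle becomes the roll of a rotator, applied additively in bone space to the input pose through a Transform (Modify) Bone node.
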AnimBlueprint /SparksCosmetics/Instrument/Guitar/LoveSpin/Meshes/LoveSpin_Guitar_AnimBP
target_skeleton=&/LoveSpin_Guitar_Skeleton
graphs [
 AnimGraph {nodes [
   n1 = "Transform (Modify) Bone - Bone: dyn_wheel"(
    ComponentPose=Local To Component(LocalPose=Input Pose())
    Alpha=1
    Translation="0,0,0"
    Rotation=Get Wheel Rotator() -> (Wheel Rotator:Rotator)
    Scale="1,1,1"
   ) {
    rotationMode=BMM_Additive
    rotationSpace=BCS_BoneSpace
    scaleMode=BMM_Additive
    scaleSpace=BCS_BoneSpace
    translationMode=BMM_Additive
    translationSpace=BCS_BoneSpace
    alphaBoolBlend {blendOption=Linear, blendOutTime=2}
    boneToModify {boneName=dyn_wheel}
   }
   Output Pose(Result=Component To Local(ComponentPose=@n1))
  ]}
 EventGraph {nodes [
   n1 = Event BlueprintUpdateAnimation()
   > n2 = Sequence()
   > Set Wheel rotation(Wheel rotation:double=FInterpTo(Current=Get Wheel rotation() -> (Wheel rotation:double), Target=-5, DeltaTime=@n1.DeltaTimeX, InterpSpeed=1)) -> (Output_Get:double)
   n3 = float * float!(A=Get Wheel rotation(), B=@n1.DeltaTimeX, C:double=30) -> (ReturnValue) {is=title K2Node_PromotableOperator}
   n4 = Get Rotator() -> (Rotator:double)
   n5 = Set Rotator(
    execute=@n2.then_1
    Rotator:double="Wrap (Float)"!(Value:double=float + float(A=@n4, B=@n3), Min:double=0, Max:double=360) -> (ReturnValue:double)
   ) -> (Output_Get:double)
   > Set Wheel Rotator(Wheel Rotator:Rotator=MakeRotator(Roll=@n5)) -> (Output_Get:Rotator)
   TryGetPawnOwner()
  ]}
]
variables [
 Rotator {type=number}
 Wheel rotation {type=number}
 Wheel Rotator {type=Rotator}
]
